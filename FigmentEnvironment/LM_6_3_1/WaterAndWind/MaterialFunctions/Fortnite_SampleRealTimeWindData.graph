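# Material function in the Fortnite Wind library that samples real-time wind data into a rotation axis and angle, plus raw color and alpha.
# One InternalSampleWindTexture call, masked by its border, supplies the wind vector; its XY length sets an ease-out angle capped by Max Rotation.
# The axis adds a cross product of Z and VectorOffset_RT2, sampled in World Iso Render space at a position defaulting to the Pivot Painter pivot.
# An unused branch blends three wind samples.
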MaterialFunction /FigmentEnvironment/LM_6_3_1/WaterAndWind/MaterialFunctions/Fortnite_SampleRealTimeWindData
graph {
 properties {bExposeToLibrary=true, libraryCategoriesText [Fortnite Wind]}
 nodes [
  CollectionParameter() {
   collection=/FigmentEnvironment/LM_6_3_1/Packages/Fortress_SharedMaterials/GlobalMaterialParameters/FortniteMaterialParameters
   parameterName=WindRenderTarget1Radius
  }
  CollectionParameter() {
   collection=&FortniteMaterialParameters
   parameterName=WindRenderTarget3Radius
  }
  CollectionParameter() {
   collection=&FortniteMaterialParameters
   parameterName=WindRenderTarget2Radius
  }
  CollectionParameter() {
   collection=&FortniteMaterialParameters
   parameterName=WindRenderTarget1Radius
  }
  CollectionParameter() {
   collection=&FortniteMaterialParameters
   parameterName=WindSpecificCameraPositionVariable1_DoNotUseElsewhere
  }
  CollectionParameter() {
   collection=&FortniteMaterialParameters
   parameterName=WindSpecificCameraPositionVariable2_DoNotUseElsewhere
  }
  CollectionParameter() {
   collection=&FortniteMaterialParameters
   parameterName=WindSpecificCameraPositionVariable3_DoNotUseElsewhere
  }
  CollectionParameter() {
   collection=&FortniteMaterialParameters
   parameterName=WindSpecificCameraPositionVariable1_DoNotUseElsewhere
  }
  n1 = MaterialFunctionCall!(Texel Specific Camera Position, OrthoWidth, Sample Point World Position, Wind Texture) {materialFunction=&/InternalSampleWindTexture}
  n2 = Multiply(A=AppendVector(A=@n1.RGB, B=@n1.RGB), B=@n1.Border Mask)
  n3 = ComponentMask(None=@n2) {b=true, g=true, r=true}
  n4 = AppendVector(A=ComponentMask(None=@n3) {g=true, r=true}, B=Constant() {r=0})
  n5 = Distance(A=@n4, B=Constant())
  n6 = FunctionInput() {inputName=Max Rotation, inputType=FunctionInput_Scalar}
  n7 = OneMinus(None=Clamp(None=Divide(A=@n5, B=@n6)))
  n8 = Multiply(A=@n7, B=@n7)
  FunctionOutput(None=Multiply(A=OneMinus(None=@n8), B=@n6)) {outputName=Rotation Angle, sortPriority=-4}
  FunctionOutput(None=@n3) {outputName=Animated Texture Output}
  n9 = FunctionInput(Preview=MaterialFunctionCall() {materialFunction=/Engine/Functions/Engine_MaterialFunctions02/PivotPainter/PivotPainter_PerObjectData}.Pivot Position) {bUsePreviewValueAsDefault=true, description=Defaults to pibot painter 1 pivot position, inputName="World Position (See Note)"}
  n10 = CollectionParameter() {
   collection=&FortniteMaterialParameters
   parameterName=World Iso Render Pivot Point
  }
  n11 = CollectionParameter() {
   collection=&FortniteMaterialParameters
   parameterName=World Iso Render Width
  }
  n12 = Divide(A=Subtract(A=ComponentMask(None=@n9) {g=true, r=true}, B=ComponentMask(None=@n10) {g=true, r=true}), B=@n11)
  n13 = CrossProduct(
   A=Constant3Vector() {constant=<rgba 0 0 1 0>}
   B=TextureSample(UVs=Subtract(A=@n12) {constB=-0.5}) {texture=/FigmentEnvironment/LM_6_3_1/WaterAndWind/Textures/VectorOffset_RT2}.RGB
  )
  FunctionOutput(None=Normalize(VectorInput=Add(A=@n4, B=@n13))) {outputName=Rotation Axis, sortPriority=-5}
  FunctionOutput(None=ComponentMask(None=@n2) {a=true}) {outputName=Animated Texture Alpha Output}
  n14 = MaterialFunctionCall() {materialFunction=&InternalSampleWindTexture}
  n15 = MaterialFunctionCall() {materialFunction=&InternalSampleWindTexture}
  n16 = MaterialFunctionCall() {materialFunction=&InternalSampleWindTexture}
  n17 = LinearInterpolate(
   A=Multiply(A=AppendVector(A=@n14.RGB, B=@n14.RGB), B=@n14.Border Mask)
   B=LinearInterpolate(A=AppendVector(A=@n15.RGB, B=@n15.RGB), B=AppendVector(A=@n16.RGB, B=@n16.RGB), Alpha=@n16.Border Mask)
   Alpha=@n15.Border Mask
  )
  n18 = ComponentMask(None=@n17) {b=true, g=true, r=true}
  n19 = AppendVector(A=ComponentMask(None=@n18) {g=true, r=true}, B=Constant() {r=0})
  n20 = Distance(A=@n19, B=Constant())
  n21 = Constant() {desc=Max rotation}
  n22 = OneMinus(None=Clamp(None=Divide(A=@n20, B=@n21)))
  n23 = Multiply(A=@n22, B=@n22)
  Multiply(A=Multiply(A=OneMinus(None=@n23), B=@n21))
  n24 = Constant3Vector() {constant=<rgba 0 0 1 0>}
  n25 = ComponentMask(None=Reroute()) {g=true, r=true}
  n26 = Constant() {r=10240}
  n27 = TextureSample(UVs=Subtract(A=Divide(A=@n25, B=@n26)) {constB=0.5}) {desc=RT2}
  Normalize(VectorInput=Add(A=@n19, B=CrossProduct(A=@n24, B=@n27.RGB)))
  Reroute(None=@n9)
  Reroute(None=ComponentMask(None=@n17) {a=true})
  Reroute(None=@n18)
  TextureObject() {texture=/FigmentEnvironment/LM_6_3_1/WaterAndWind/Textures/WindRenderTarget1}
  TextureObject() {desc=RT3}
  TextureObject() {desc=RT1}
  TextureObject() {desc=RT2}
 ]
}
